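# Physics asset for the Punk Koi medium female backpack cosmetic, defining capsule bodies and joints for its bow and zipper bones.
# The bow master bone is constrained to the pack body, and the two bow pieces each constrain to that master, forming a small branched chain with tight angular limits.
# The zipper constrains directly to the pack with much wider swing limits and a narrow twist, letting it dangle more freely than the bow.
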PhysicsAsset /BRCosmetics/Accessories/FORT_Backpacks/Backpack_F_MED_Punk_Koi/Meshes/F_MED_Punk_Koi_Pack_PhysicsDef
bodies {
 dyn_bow_01 [{shapeType=Capsule, length=3, radius=2, shapeName=dyn_bow_01_capsule, center=<2.4607 0.45 0.0629>, rotation=<rot 0.2264 -90 0.8824>}]
 dyn_bow_02 [{shapeType=Capsule, length=3, radius=2, shapeName=dyn_bow_02_capsule, center=<2.5341 0.25 -0.1853>, rotation=<rot 0.7079 -90 -0.3056>}]
 dyn_bow_MSTR [{shapeType=Capsule, length=1, radius=1, shapeName=dyn_bow_MSTR_capsule, center=<0.239 0 0>, rotation=<rot 0 0 23.5865>}]
 dyn_zipper [{
   shapeType=Capsule
   length=0.505
   radius=0.505
   shapeName=dyn_zipper_capsule
   center=<1.4919 0.0008 -0.0136>
   rotation=<rot 0 89.5029 -89.6458>
  }]
 pack [{shapeType=Capsule, length=17, radius=4, center=<0 -4.75 7.2852>, rotation=<rot 90 0 0>}]
}
constraints [
 {bone1Name=dyn_bow_MSTR, bone2Name=pack, swing1LimitDegrees=1, swing2LimitDegrees=5, twistLimitDegrees=1}
 {bone1Name=dyn_bow_02, bone2Name=dyn_bow_MSTR, swing1LimitDegrees=9, swing2LimitDegrees=1, twistLimitDegrees=1}
 {bone1Name=dyn_bow_01, bone2Name=dyn_bow_MSTR, swing1LimitDegrees=9, swing2LimitDegrees=1, twistLimitDegrees=1}
 {bone1Name=dyn_zipper, bone2Name=pack, swing1LimitDegrees=20, swing2LimitDegrees=60, twistLimitDegrees=1}
]
